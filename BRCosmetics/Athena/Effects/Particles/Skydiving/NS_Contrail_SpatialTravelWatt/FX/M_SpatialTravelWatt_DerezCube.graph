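# Glowing material for Niagara sprites, ribbons and mesh particles, named for the derez cubes of the SpatialTravelWatt skydiving contrail.
# A dynamic parameter blends the Color parameter toward particle color and drives Fresnel brightness.
# The Fresnel term is remapped to HDR, tints that color, and is added to the base color before MF_FX_EmissiveAdjustments.
# The GridBrightness and CubeLerp channels are declared but unused in the graph.
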Material /BRCosmetics/Athena/Effects/Particles/Skydiving/NS_Contrail_SpatialTravelWatt/FX/M_SpatialTravelWatt_DerezCube
graph {
 properties {
  bAutomaticallySetUsageInEditor=false
  bUsedWithNiagaraMeshParticles=true
  bUsedWithNiagaraRibbons=true
  bUsedWithNiagaraSprites=true
  maxWorldPositionOffsetDisplacement=800
  usedShadingModels=Default Lit
 }
 nodes [
  Constant() {r=0.5}
  Constant() {r=0.2}
  n1 = DynamicParameter() {defaultValue=<rgba 0 0 1 0>, paramNames [FresnelGlow, UseParticleColor, GridBrightness, CubeLerp]}
  n2 = LinearInterpolate(A=VectorParameter() {parameterName=Color, defaultValue=<rgba 0.0342 0.4499 1 1>}.RGB, B=ParticleColor().RGB, Alpha=@n1.UseParticleColor)
  n3 = Multiply(B=@n2) {constA=1}
  n4 = Constant() {r=2}
  n5 = NamedRerouteDeclaration(None=MaterialFunctionCall(
    LDR Input=MaterialFunctionCall(Brightness=@n1.FresnelGlow, Exponent=@n4) {materialFunction=&MF_Fresnel_FX}.Fresnel
    HDR Dynamic Range=Constant() {r=1}
    HDR Intensity Multiplier=Constant() {r=100}
   ) {materialFunction=&DeriveHDRfromLDR}.RemappedRange) {name=Fresnel, nodeColor=<rgba 1 0 0.6797 1>}
  MaterialFunctionCall(Emissive=Add(A=Multiply(B=@n3) {constA=1}, B=Multiply(A=@n2, B=@n5))) {materialFunction=&MF_FX_EmissiveAdjustments}
 ]
}
