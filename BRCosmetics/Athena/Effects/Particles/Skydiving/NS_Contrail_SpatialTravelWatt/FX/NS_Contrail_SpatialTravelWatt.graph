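# Niagara skydiving contrail cosmetic that trails glowing blue spiral ribbons from character sockets, with separate hand and foot tuning.
# NE_DerezRibbons burst-spawns fixed-count ribbon points and repositions them each update using lag, sine/cosine wobble and owner velocity.
# NE_DerezCubes and NE_Glow follow those ribbons through particle reads, adding orbiting mesh cubes, flare sprites, lights and jet meshes.
# Smoothed lean, fade-in, front-end and time-of-day modifiers adjust shape or brightness.
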NiagaraSystem /BRCosmetics/Athena/Effects/Particles/Skydiving/NS_Contrail_SpatialTravelWatt/FX/NS_Contrail_SpatialTravelWatt
dependencies {
 usedDataInterfaces [
  &NiagaraDataInterfaceCamera
  &NiagaraDataInterfaceCurve
  &NiagaraDataInterfaceColorCurve
  &NiagaraDataInterfaceArrayInt32
  &NiagaraDataInterfaceParticleRead
  &NiagaraDataInterfacePlatformSet
  &NiagaraDataInterfaceSocketReader
 ]
 usedDynamicInputs [
  &MakeIntFromFloat
  &FN_PickFloatDependOnPlatform
  &SetBoolByFloatComparison
 ]
 usedModules [
  &self:SystemSpawnScript.SystemScriptSource.SystemScriptGraph.NiagaraNodeAssignment_0.SetVariables_#1
  &self:SystemSpawnScript.SystemScriptSource.SystemScriptGraph.NiagaraNodeAssignment_1.SetVariables_#2
  &self:SystemSpawnScript.SystemScriptSource.SystemScriptGraph.NiagaraNodeAssignment_2.SetVariables_#3
  &self:SystemSpawnScript.SystemScriptSource.SystemScriptGraph.NiagaraNodeAssignment_5.SetVariables_#4
  &SystemState
  &self:SystemSpawnScript.SystemScriptSource.SystemScriptGraph.NiagaraNodeAssignment_3.SetVariables_#5
  &self:SystemSpawnScript.SystemScriptSource.SystemScriptGraph.NiagaraNodeAssignment_4.SetVariables_#6
  &Timeline
 ]
}
emitters {
 NE_DerezCubes {
  emitterUpdateScript [
   &EmitterState()
   &SpawnRate(SpawnRate=&Lerp_Float(A=20, B=20, Alpha=$System.SmoothLeanAlpha), Use Spawn Probability=false)
   &SpawnRate(SpawnGroup=1, SpawnRate=&Lerp_Float(A=0, B=60, Alpha=$System.SmoothLeanAlpha), Use Spawn Probability=false) {module=SpawnRate001}
  ]
  particleSpawnScript [
   &PartitionParticles(Number Of Partitions=4, Partition Mode:ENiagara_PartitionMode_v2=::NewEnumerator5)
   /SetVariables(Particles.PartitionParticles.Partition:NiagaraInt32=&SelectIntFromArray(
     Array Sampling Mode=::NewEnumerator1
     Direct Array Index=?
     Int Selection Array=DataInterface() {propertyValues {IntData [0,0,0,1,1,1,2,3]}}
     Random Seed=0
    )) {isSetParameters=true}
   /SetVariables(Particles.HandFootAlpha:NiagaraFloat=&MakeCustomFloatFromBool(
     Bool=&SetBoolByIntComparison(A=$Particles.PartitionParticles.Partition, B=1, Comparison Type=::NewEnumerator0)
     True Float=1
     False Float=0
    )) {isSetParameters=true}
   &InitializeParticle(
    Color Mode=::NewEnumerator0
    Lifetime=&MaskFloatBySpawnGroup(
     Masked Float=&Lerp_Float(
      A=&RandomRangeFloat(Minimum=1.1, Maximum=1.3)
      B=&RandomRangeFloat(Minimum=0.9, Maximum=1.1)
      Alpha=$System.SmoothLeanAlpha
     )
     Passthrough Float=&RandomRangeFloat(Minimum=0.25, Maximum=0.4)
    )
    Mesh Scale Mode=::NewEnumerator3
    Mesh Uniform Scale=&Multiply_Float(
     A=&Multiply_Float(
      A=&MaskFloatBySpawnGroup(
       Masked Float=&RandomRangeFloat(Minimum=2, Maximum=3)
       Passthrough Float=&RandomRangeFloat(Minimum=0.5, Maximum=1)
      )
      B=&Lerp_Float(A=1, B=1.5, Alpha=$System.SmoothLeanAlpha)
     )
     B=&Lerp_Float(A=1, B=0.5, Alpha=$Particles.HandFootAlpha)
    )
    Sprite Size Mode=::NewEnumerator3
    Uniform Sprite Size=&RandomRangeFloat(Minimum=5, Maximum=7)
    UsePositionOffset=false
    Write Material Random=true
   )
   &InitialMeshOrientation(Use Rotation Vector=false)
   /SetVariables(Particles.InitialSystemAge:NiagaraFloat=$System.Age) {isSetParameters=true}
  ]
  particleUpdateScript [
   &ParticleState()
   &self:FollowRibbon(
    Alpha:NiagaraFloat=&Lerp_Float(
     A=&MaskFloatBySpawnGroup(Masked Float=&RandomRangeFloat(Minimum=0, Maximum=0), Passthrough Float=0)
     B=&MaskFloatBySpawnGroup(Masked Float=0.8, Passthrough Float=0.15)
     Alpha=&FloatFromCurve(FloatCurve=DataInterface() {propertyValues {Curve {keys [<at 0 0>,<at 1 1>]}}})
    )
    ParticlePerRibbon:NiagaraInt32=$System.ParticlePerRibbon
    RibbonNumber:NiagaraInt32=$Particles.PartitionParticles.Partition
    RibbonReader=DataInterface() {propertyValues {EmitterBinding {bindingMode=Other, emitterName=NE_DerezRibbons}}}
   ) {module=FollowRibbon}
   &RotateAroundPoint(
    Radius:NiagaraFloat=&MaskFloatBySpawnGroup(
     Masked Float=&Multiply_Float(
      A=&Lerp_Float(A=15, B=8, Alpha=$Particles.HandFootAlpha)
      B=&Lerp_Float(
       A=0.6
       B=1.2
       Alpha=&FloatFromCurve(FloatCurve=DataInterface() {propertyValues {Curve {keys [<at 0 0>,<at 1 1>]}}})
      )
     )
     Passthrough Float=&Lerp_Float(A=10, B=5, Alpha=$Particles.HandFootAlpha)
    )
    Rotation Center:NiagaraPosition=$Output.FollowRibbon.RibbonPos
    Rotation Phase:NiagaraFloat=&Add_Float(
     A=&Multiply_Float(A=$Particles.NormalizedAge, B=0.2)
     B=&Multiply_Float(A=$StackContext.InitialSystemAge, B=1)
    )
    Set Particle Position=false
   )
   /SetVariables(Particles.Position=$Output.RotateAroundPoint.OutputPosition) {isSetParameters=true}
   /SetVariables(Particles.SpritePos:NiagaraPosition=&AddVectorToPosition(Position=$Particles.Position, Vector=&RandomVector(Vector Scale=10))) {isSetParameters=true}
   &Color(Color=&Multiply_LinearColorByFloat(
     Linear Color=&ColorFromCurve(
      ColorCurve=DataInterface() {propertyValues {
        AlphaCurve {keys [<at 0 1 RCIM_Cubic>]}
        BlueCurve {keys [<at 0.464 0.8 RCIM_Cubic>,<at 0.5032 0.5 RCIM_Cubic>]}
        GreenCurve {keys [<at 0.464 0.6116 RCIM_Cubic>,<at 0.5032 0.2057 RCIM_Cubic>]}
        RedCurve {keys [<at 0.464 0 RCIM_Cubic>,<at 0.5032 0 RCIM_Cubic>]}
       }}
      CurveIndex=&RandomRangeFloat(Minimum=0, Maximum=1)
     )
     Float=&Multiply_Float(
      A=&Lerp_Float(
       A=&MaskFloatBySpawnGroup(Masked Float=1, Passthrough Float=3)
       B=&RandomRangeFloat(Minimum=2, Maximum=2)
       Alpha=&FloatFromCurve(FloatCurve=DataInterface() {propertyValues {Curve {keys [<at 0 1>,<at 1 0>]}}})
      )
      B=&FNTimeOfDayScalar(Night Scale=0.5, Morning Scale=0.5, Evening Scale=0.5)
     )
    ))
   &ScaleMeshSize(Scale Factor=&VectorFromFloat(Value=&Multiply_Float(
      A=&FloatFromCurve(
       FloatCurve=DataInterface() {propertyValues {Curve {keys [
           {arriveTangent=2.0194, interpMode=RCIM_Cubic, leaveTangent=2.0194, tangentMode=RCTM_User, time=0, value=0}
           {arriveTangent=-0.0389, interpMode=RCIM_Cubic, leaveTangent=-0.0389, tangentMode=RCTM_User, time=0.8, value=1}
           {arriveTangent=-2.6037, interpMode=RCIM_Cubic, leaveTangent=-2.6037, tangentMode=RCTM_User, time=1, value=0}
          ]}}}
       Scale Curve=$System.FadeIn
      )
      B=&ClampFloat(Float=&Divide_Float(
        A=&Subtract_Float(
         A=/Niagara/DynamicInputs/Camera/DistanceToCamera(Use Owner Position:NiagaraBool=true, Camera Query:NiagaraDataInterfaceCamera=DataInterface())
         B=50
        )
        B=100
       ))
     )))
   &DynamicMaterialParameters(
    Index 0 Param 2=1
    Module.Index0Float0Write=true
    Module.Index0Float1Write=true
    Module.Index0Float2Write=true
    Module.Index0Float3Write=true
   )
  ]
  rendererClasses [&NiagaraMeshRendererProperties]
  renderers [{rendererClass=&NiagaraMeshRendererProperties, rendererIndex=0}]
 }
 NE_DerezRibbons {
  emitterSpawnScript [/SetVariables() {isSetParameters=true}]
  emitterUpdateScript [
   &EmitterState(
    Inactive Response=::NewEnumerator1
    Life Cycle Mode=::NewEnumerator1
    Loop Behavior=::NewEnumerator1
    Loop Duration Mode=::NewEnumerator1
    UseLoopDelay=false
   )
   &SpawnBurst_Instantaneous(
    Spawn Count=&Multiply_Int(A=$System.RibbonNumber, B=$System.ParticlePerRibbon)
    Use Spawn Probability=false
    UseLoopCountLimit=false
   )
   /SetVariables(Emitter.CustomTime:NiagaraFloat=&Add_Float(A=$Emitter.CustomTime, B=&Lerp_Float(A=0.01, B=0.025, Alpha=$System.SmoothLeanAlpha))) {isSetParameters=true}
  ]
  particleSpawnScript [
   &PartitionParticles(
    Normalized Index In Partition Includes One=true
    Number Of Partitions=$System.RibbonNumber
    Particles Per Partition:NiagaraInt32=$System.ParticlePerRibbon
    Partition Mode:ENiagara_PartitionMode_v2=::NewEnumerator1
    Write Partition Index to Ribbon Link Order:ENiagara_AttributeSamplingApplyOutput=::NewEnumerator0
    Write Partition to Ribbon ID=::NewEnumerator0
   )
   /SetVariables(Particles.HandFootAlpha:NiagaraFloat=&MakeCustomFloatFromBool(
     Bool=&SetBoolByIntComparison(A=$Particles.PartitionParticles.Partition, B=1, Comparison Type=::NewEnumerator0)
     True Float=1
     False Float=0
    )) {isSetParameters=true}
   &InitializeParticle(
    Color Mode=::NewEnumerator0
    Lifetime=5
    Ribbon Width=&Lerp_Float(A=15, B=6, Alpha=$Particles.HandFootAlpha)
    Ribbon Width Mode=::NewEnumerator1
    UsePositionOffset=false
    Write Material Random=true
   )
   &SocketLocation(
    SamplePosOffset=false
    Socket Index=$Particles.PartitionParticles.Partition
    Socket Reader=$System.Socket Reader
    Socket Sampling Mode=::NewEnumerator3
    VelocitySpeedLimit=false
   ) {module=SocketLocation001}
   /SetVariables(Particles.RibbonU0Override=$Particles.PartitionParticles.Normalized Index In Partition) {isSetParameters=true}
   /SetVariables() {isSetParameters=true}
  ]
  particleUpdateScript [
   &ParticleState(Kill Particles When Lifetime Has Elapsed=false)
   /SetVariables(Particles.SpiralTime:NiagaraFloat=&Add_Float(
     A=&Multiply_Float(A=$Particles.PartitionParticles.NormalizedPartition, B=1)
     B=&Add_Float(
      A=&Multiply_Float(A=$Emitter.CustomTime, B=0.5)
      B=&Lerp_Float(
       A=0
       B=&Multiply_Float(
        A=&Lerp_Float(A=0.5, B=0.5, Alpha=$System.SmoothLeanAlpha)
        B=&Lerp_Float(A=1, B=0.25, Alpha=$Particles.HandFootAlpha)
       )
       Alpha=$Particles.PartitionParticles.Normalized Index In Partition
      )
     )
    )) {isSetParameters=true}
   /SetVariables(Particles.RibbonFacing=&MakeVector(
     X=&Sine(Normalized Angle=$Particles.SpiralTime, Period=1, Scale=1, Bias=0)
     Y=&Cosine(Normalized Angle=$Particles.SpiralTime, Period=1, Scale=1, Bias=0)
     Z=0
    )) {isSetParameters=true}
   &SocketLocation(
    Position Sampling=::NewEnumerator1
    SamplePosOffset=true
    Sampled Position Offset=&Lerp_Vector(
     A=&Lerp_Vector(A=<0 0 20>, B=<0 0 35>, Alpha=$System.SmoothLeanAlpha)
     B=&Lerp_Vector(A=<0 0 10>, B=<0 0 40>, Alpha=$System.SmoothLeanAlpha)
     Alpha=$Particles.HandFootAlpha
    )
    Sampled Position Offset Coordinate Space:ENiagaraSkelMeshTransforms=::NewEnumerator1
    Socket Index=$Particles.PartitionParticles.Partition
    Socket Reader=$System.Socket Reader
    Socket Sampling Mode=::NewEnumerator3
    VelocitySpeedLimit=false
   )
   /SetVariables(Particles.GoalPos:NiagaraPosition=&AddVectorToPosition(
     Position=&AddVectorToPosition(
      Position=Linked() {linkedVariable {name=Output.SocketLocation.SampledPosition, type=Vector3f}}
      Vector=&Lerp_Vector(
       A=<0 0 0>
       B=&Add_Vector(
        A=&Multiply_VectorByFloat(
         Vector=&Multiply_VectorByFloat(
          Vector=&Add_Vector(
           A=&Lerp_Vector(A=<0 0 150>, B=<0 0 250>, Alpha=$System.SmoothLeanAlpha)
           B=&FN_FrontEndVector3Modifier(Gameplay=<0 0 0>, Front End=&Lerp_Vector(A=<0 -150 0>, B=<0 -250 0>, Alpha=$System.SmoothLeanAlpha))
          )
          Float=&Lerp_Float(A=1, B=0.3, Alpha=$Particles.HandFootAlpha)
         )
         Float=&Lerp_Float(A=0.1, B=1, Alpha=$System.FadeIn)
        )
        B=&Lerp_Vector(
         A=<0 0 0>
         B=&Lerp_Vector(A=<0 -100 0>, B=<0 0 0>, Alpha=$System.SmoothLeanAlpha)
         Alpha=&Abs_Float(Float:NiagaraFloat=&DotProduct(Vector A=&ReturnCameraVector(Camera Vectors=::NewEnumerator1, Coordinate Space=::Simulation), Vector B=<0 0 1>))
        )
       )
       Alpha=$Particles.PartitionParticles.Normalized Index In Partition
      )
     )
     Vector=&FN_FrontEndVector3Modifier(
      Gameplay=<0 0 0>
      Front End=&Lerp_Vector(
       A=<0 0 0>
       B=&Lerp_Vector(A=<0 -30 -10>, B=<0 -20 -35>, Alpha=$Particles.HandFootAlpha)
       Alpha=$System.SmoothLeanAlpha
      )
     )
    )) {isSetParameters=true}
   /SetVariables(Particles.Position=&AddVectorToPosition(
     Position=&Lerp_Vector(
      A=Linked() {linkedVariable {name=Particles.GoalPos, type=NiagaraPosition}}
      B=&Lerp_Vector(
       A=Linked() {linkedVariable {name=Particles.Position, type=NiagaraPosition}}
       B=Linked() {linkedVariable {name=Particles.GoalPos, type=NiagaraPosition}}
       Alpha=&Multiply_Float(
        A=0.02
        B=&Lerp_Float(
         A=10
         B=8
         Alpha=&FloatFromCurve(
          FloatCurve=DataInterface() {propertyValues {Curve {keys [
              {arriveTangent=0.1322, interpMode=RCIM_Cubic, leaveTangent=0.1322, tangentMode=RCTM_User, time=0, value=0}
              {arriveTangent=2.4891, interpMode=RCIM_Cubic, leaveTangent=2.4891, tangentMode=RCTM_User, time=1, value=1}
             ]}}}
          CurveIndex=$Particles.PartitionParticles.Normalized Index In Partition
         )
        )
       )
      )
      Alpha=&FloatFromCurve(
       FloatCurve=DataInterface() {propertyValues {Curve {keys [
           {arriveTangent=2.1246, interpMode=RCIM_Cubic, leaveTangent=2.1246, tangentMode=RCTM_User, time=0, value=0.2}
           {interpMode=RCIM_Cubic, tangentMode=RCTM_User, time=1, value=1}
          ]}}}
       CurveIndex=$Particles.PartitionParticles.Normalized Index In Partition
      )
     )
     Vector=&Multiply_VectorByFloat(
      Vector=&MakeVector(
       X=&Sine(Normalized Angle=$Particles.SpiralTime, Period=0.4, Scale=1, Bias=0)
       Y=&Cosine(Normalized Angle=$Particles.SpiralTime, Period=0.4, Scale=1, Bias=0)
       Z=0
      )
      Float=&Multiply_Float(
       A=&Multiply_Float(
        A=&Multiply_Float(
         A=&Lerp_Float(A=5, B=2, Alpha=$Particles.HandFootAlpha)
         B=&FN_FrontEndFloatModifier(Gameplay=1, Front End=0.6)
        )
        B=$System.FadeIn
       )
       B=&FloatFromCurve(
        FloatCurve=DataInterface() {propertyValues {Curve {keys [
            {arriveTangent=2.1246, interpMode=RCIM_Cubic, leaveTangent=2.1246, tangentMode=RCTM_User, time=0, value=0}
            {interpMode=RCIM_Cubic, tangentMode=RCTM_User, time=0.5, value=1}
            <at 1 0 RCIM_Cubic>
           ]}}}
        CurveIndex=$Particles.PartitionParticles.Normalized Index In Partition
       )
      )
     )
    )) {isSetParameters=true}
   /SetVariables(Particles.Position=&AddVectorToPosition(
     Position=$Particles.Position
     Vector=&Multiply_Vector(
      A=&TransformVector(Vector=$Engine.Owner.Velocity, Source Space=::World, Destination Space=::Local)
      B=&Multiply_VectorByFloat(
       Vector=<-1 -1 0>
       Float=&Lerp_Float(
        A=&FloatFromCurve(
         FloatCurve=DataInterface() {propertyValues {Curve {keys [
             {arriveTangent=2.7932, interpMode=RCIM_Cubic, leaveTangent=2.7932, tangentMode=RCTM_User, time=0, value=0}
             {interpMode=RCIM_Cubic, tangentMode=RCTM_User, time=1, value=1}
            ]}}}
         CurveIndex=$Particles.PartitionParticles.Normalized Index In Partition
         Scale Curve=0.02
        )
        B=&FloatFromCurve(
         FloatCurve=DataInterface() {propertyValues {Curve {keys [
             {arriveTangent=2.7932, interpMode=RCIM_Cubic, leaveTangent=2.7932, tangentMode=RCTM_User, time=0.1, value=0}
             {interpMode=RCIM_Cubic, tangentMode=RCTM_User, time=1, value=1}
            ]}}}
         CurveIndex=$Particles.PartitionParticles.Normalized Index In Partition
         Scale Curve=0.01
        )
        Alpha=$Particles.HandFootAlpha
       )
      )
     )
    )) {isSetParameters=true}
   &ScaleRibbonWidth(Ribbon Width Scale=&Multiply_Float(
     A=&Multiply_Float(
      A=&FloatFromCurve(
       FloatCurve=DataInterface() {propertyValues {Curve {keys [
           {arriveTangent=3.0384, interpMode=RCIM_Cubic, leaveTangent=3.0384, tangentMode=RCTM_User, time=0, value=0.3}
           {arriveTangent=-0.0854, interpMode=RCIM_Cubic, leaveTangent=-0.0854, tangentMode=RCTM_User, time=0.7, value=1.2}
           {arriveTangent=-5.7045, interpMode=RCIM_Cubic, leaveTangent=-5.7045, tangentMode=RCTM_User, time=1, value=0.5}
          ]}}}
       CurveIndex=$Particles.PartitionParticles.Normalized Index In Partition
      )
      B=&Add_Float(
       A=1
       B=&Sine(
        Normalized Angle=&Add_Float(
         A=&Add_Float(
          A=&Multiply_Float(A=$Emitter.CustomTime, B=-2)
          B=&Multiply_Float(A=$Particles.PartitionParticles.Normalized Index In Partition, B=1)
         )
         B=&Multiply_Float(A=$Particles.PartitionParticles.NormalizedPartition, B=0.556)
        )
        Period=1
        Scale=0.3
        Bias=0
       )
      )
     )
     B=$System.FadeIn
    ))
   /SetVariables(Particles.SmoothVelocity:Vector3f=&Lerp_Vector(A=$Particles.SmoothVelocity, B=$Engine.Owner.Velocity, Alpha=0.1)) {isSetParameters=true}
   &InheritSourceMovement(
    Applied Position Scale=&Multiply_VectorByFloat(
     Vector=<-1 -1 0>
     Float=&Lerp_Float(
      A=&FloatFromCurve(
       FloatCurve=DataInterface() {propertyValues {Curve {keys [
           {arriveTangent=2.7932, interpMode=RCIM_Cubic, leaveTangent=2.7932, tangentMode=RCTM_User, time=0, value=0}
           {interpMode=RCIM_Cubic, tangentMode=RCTM_User, time=1, value=1}
          ]}}}
       CurveIndex=$Particles.PartitionParticles.Normalized Index In Partition
      )
      B=&FloatFromCurve(
       FloatCurve=DataInterface() {propertyValues {Curve {keys [
           {arriveTangent=2.7932, interpMode=RCIM_Cubic, leaveTangent=2.7932, tangentMode=RCTM_User, time=0.1, value=0}
           {interpMode=RCIM_Cubic, tangentMode=RCTM_User, time=1, value=1}
          ]}}}
       CurveIndex=$Particles.PartitionParticles.Normalized Index In Partition
       Scale Curve=0.5
      )
      Alpha=$Particles.HandFootAlpha
     )
    )
    ApplyForce=false
    ApplyPosition=true
    ApplyPositionFalloff=true
    ApplyPositionFalloffCurve=false
    Source Velocity:Vector3f=&TransformVector(Vector=$Engine.Owner.Velocity, Source Space=::World, Destination Space=::Local)
   ) {enabled=false}
   &SolveForcesAndVelocity(Clamp Velocity=false, Limit Acceleration=false)
   &Color(Color=&Multiply_LinearColorByFloat(
     Linear Color=&ColorFromCurve(
      ColorCurve=DataInterface() {propertyValues {
        AlphaCurve {keys [<at 0.5942 1 RCIM_Cubic>,<at 1 0 RCIM_Cubic>]}
        BlueCurve {keys [<at 0 5 RCIM_Cubic>,<at 0.1475 1 RCIM_Cubic>]}
        GreenCurve {keys [<at 0 5 RCIM_Cubic>,<at 0.1475 1 RCIM_Cubic>]}
        RedCurve {keys [<at 0 5 RCIM_Cubic>,<at 0.1475 1 RCIM_Cubic>]}
       }}
      CurveIndex=$Particles.PartitionParticles.Normalized Index In Partition
     )
     Float=&FN_FrontEndFloatModifier(
      Gameplay=&FNTimeOfDayScalar(Night Scale=0.5, Morning Scale=0.3, Evening Scale=0.3)
      Front End=&Lerp_Float(A=0.5, B=0.3, Alpha=$System.SmoothLeanAlpha)
     )
    ))
   /SetVariables(Particles.GlowRibbonWidth:NiagaraFloat=&Multiply_Float(A=$Particles.RibbonWidth, B=3)) {isSetParameters=true}
   /SetVariables(Particles.GlowRibbonColor:LinearColor=&MultiplyLinearColor(
     Linear Color=&ColorFromCurve(
      ColorCurve=DataInterface() {propertyValues {
        AlphaCurve {keys [<at 0 0.5 RCIM_Cubic>,<at 1 0.5 RCIM_Cubic>]}
        BlueCurve {keys [<at 0 1 RCIM_Cubic>,<at 1 0.5 RCIM_Cubic>]}
        GreenCurve {keys [<at 0 0.6376 RCIM_Cubic>,<at 1 0.0707 RCIM_Cubic>]}
        RedCurve {keys [<at 0 0.0823 RCIM_Cubic>,<at 1 0 RCIM_Cubic>]}
       }}
      CurveIndex=$Particles.PartitionParticles.Normalized Index In Partition
     )
     Scale Factor=&MakeVector4_FromLinearColor(LinearColor=$Particles.Color)
    )) {isSetParameters=true}
   &DynamicMaterialParameters(
    Index 0 Param 1=$System.SmoothLeanAlpha
    Module.Index0Float0Write=true
    Module.Index0Float1Write=true
    Module.Index0Float2Write=true
    Module.Index0Float3Write=true
   )
  ]
  rendererClasses [&NiagaraRibbonRendererProperties]
  renderers [
   {rendererClass=&NiagaraRibbonRendererProperties}
   {rendererClass=&NiagaraRibbonRendererProperties, rendererIndex=1}
  ]
 }
 NE_Glow {
  emitterUpdateScript [
   &EmitterState(
    Inactive Response=::NewEnumerator1
    Life Cycle Mode=::NewEnumerator1
    Loop Behavior=::NewEnumerator1
    Loop Duration Mode=::NewEnumerator1
    UseLoopDelay=false
   )
   &SpawnBurst_Instantaneous(Spawn Count=$System.RibbonNumber, Use Spawn Probability=false, UseLoopCountLimit=false)
   /SetVariables(Emitter.Oscillation:NiagaraFloat=&Waveform(
     Waveform Count=::NewEnumerator1
     "Waveform [1]"=::NewEnumerator0
     "[1] Frequency"=2.3541
     "[1] Amplitude Min/Max"=<0 1>
     "[2] Average"=false
     "[2] Input Position"=0
     "[2] Phase"=$Emitter.Age
     "[2] Phase Scale"=1
     "[2] Compound Frequency"=2
     "[2] Frequency"=2
     "[2] Amplitude Min/Max"=<0 1>
     "[2] Scale"=1
     "[2] Bias"=0
     "[2] Exponent"=1
     "[2] Pulse Duration"=0.1
     "[2] Clip Triangle Min/Max"=<0 0>
     "[2] Random Seed"=0
    )) {isSetParameters=true}
  ]
  particleSpawnScript [
   &PartitionParticles(Number Of Partitions=4)
   /SetVariables(Particles.HandFootAlpha:NiagaraFloat=&MakeCustomFloatFromBool(
     Bool=&SetBoolByIntComparison(A=$Particles.PartitionParticles.Partition, B=1, Comparison Type=::NewEnumerator0)
     True Float=1
     False Float=0
    )) {isSetParameters=true}
   &InitializeParticle(
    Lifetime=0.25
    Mass Mode=::NewEnumerator1
    Mesh Scale Mode=::NewEnumerator3
    Mesh Uniform Scale=&Lerp_Float(A=1.2, B=0.7, Alpha=$Particles.HandFootAlpha)
    Position Mode=::NewEnumerator1
    Sprite Rotation Mode=::NewEnumerator2
    Sprite Size=&Vector2DFromFloat(Value=&Lerp_Float(A=50, B=25, Alpha=$Particles.HandFootAlpha))
    Sprite Size Mode=::NewEnumerator4
    UsePositionOffset=true
    Write Material Random=true
   ) {module=InitializeParticle001}
   &CameraOffset()
   /SetVariables(Particles.GlowCameraOffset=10) {isSetParameters=true}
  ]
  particleUpdateScript [
   &ParticleState(Kill Particles When Lifetime Has Elapsed=false)
   &ScaleMeshSize(Scale Factor=&VectorFromFloat(Value=&FloatFromCurve(
      FloatCurve=DataInterface() {propertyValues {Curve {keys [
          {interpMode=RCIM_Cubic, tangentMode=RCTM_User, time=0, value=0}
          {arriveTangent=0.0411, interpMode=RCIM_Cubic, leaveTangent=0.0411, tangentMode=RCTM_User, time=0.2, value=1.25}
          {arriveTangent=-0.1314, interpMode=RCIM_Cubic, leaveTangent=-0.1314, tangentMode=RCTM_User, time=0.5, value=1}
         ]}}}
      CurveIndex=$Engine.System.Age
     )))
   &SocketLocation(
    Orientation Sampling=::NewEnumerator1
    SamplePosOffset=true
    Sampled Position Offset=&Lerp_Vector(A=<0 0 15>, B=<0 0 0>, Alpha=$Particles.HandFootAlpha)
    Sampled Position Offset Coordinate Space:ENiagaraSkelMeshTransforms=::NewEnumerator1
    Socket Index=$Particles.PartitionParticles.Partition
    Socket Reader=$System.Socket Reader
    Socket Sampling Mode=::NewEnumerator3
    VelocitySpeedLimit=false
   ) {enabled=false}
   &SolveForcesAndVelocity(Clamp Velocity=false, Limit Acceleration=false)
   &self:FollowRibbon(
    ParticlePerRibbon:NiagaraInt32=$System.ParticlePerRibbon
    RibbonNumber:NiagaraInt32=$Particles.PartitionParticles.Partition
    RibbonReader=DataInterface() {propertyValues {EmitterBinding {bindingMode=Other, emitterName=NE_DerezRibbons}}}
   ) {module=FollowRibbon}
   /SetVariables(Particles.Position=&AddVectorToPosition(
     Position=$Output.FollowRibbon.RibbonPos
     Vector=&Multiply_VectorByFloat(Vector=$Output.FollowRibbon.DirectionVector, Float=-2)
    )) {isSetParameters=true}
   &ScaleColor(
    Scale Alpha=&FloatFromCurve(
     FloatCurve=DataInterface() {propertyValues {Curve {keys [
         {arriveTangent=6.4019, interpMode=RCIM_Cubic, leaveTangent=6.4019, tangentMode=RCTM_User, time=0, value=0}
         {interpMode=RCIM_Cubic, leaveTangent=0.002, time=0.2, value=0.7}
        ]}}}
     Scale Curve=0.1
    )
    Scale RGB=&MakeVectorFromLinearColorRGB(Color=<rgba 0 0.6604 1 1>)
    ScaleA=true
    ScaleRGB=true
   )
   &ScaleSpriteSize(Scale Factor=&Multiply_Vector2D_ByFloat(
     Float=&Multiply_Float(
      A=&Multiply_Float(
       A=&Multiply_Float(
        A=&Multiply_Float(
         A=&FloatFromCurve(
          FloatCurve=DataInterface() {propertyValues {Curve {keys [
              {arriveTangent=0.7585, interpMode=RCIM_Cubic, leaveTangent=0.7585, tangentMode=RCTM_User, time=0, value=0.4}
              {interpMode=RCIM_Cubic, tangentMode=RCTM_User, time=1, value=1}
             ]}}}
          Scale Curve=$System.FadeIn
         )
         B=1
        )
        B=&Lerp_Float(
         A=1
         B=3
         Alpha=&FloatFromCurve(
          FloatCurve=DataInterface() {propertyValues {Curve {keys [
              {interpMode=RCIM_Cubic, tangentMode=RCTM_User, time=0, value=0}
              {interpMode=RCIM_Cubic, tangentMode=RCTM_User, time=0.1, value=1}
              {interpMode=RCIM_Cubic, tangentMode=RCTM_User, time=0.9, value=0}
             ]}}}
          CurveIndex=$System.Timeline.NormalizedTime
         )
        )
       )
       B=&Lerp_Float(A=1, B=1.2, Alpha=&Multiply_Float(A=$Emitter.Oscillation, B=$System.SmoothLeanAlpha))
      )
      B=&FN_FrontEndFloatModifier(Gameplay=1, Front End=0.7)
     )
     Vector2D=<1 1>
    ))
   &SpriteRotationRate(Rotation Rate=&UniformRangedFloat(Minimum=-50, Maximum=50))
   &SpriteFacingAndAlignment(
    Sprite Alignment=&TransformVector(Vector=<1 0 0>, Source Space=::World, Destination Space=::Simulation)
    WriteAlignment=true
    WriteFacing=false
   )
   /SetVariables(Particles.DarkGlowSize:Vector2f=&Multiply_Vector2D_ByFloat(Float=1.3, Vector2D=$Particles.SpriteSize)) {isSetParameters=true}
   /SetVariables(Particles.DargGlowColor:LinearColor=&MakeLinearColorFromVectorAndFloat(
     "Vector (RGB)"=&MakeVectorFromLinearColorRGB(Color=<rgba 0 0.6899 1 1>)
     "Float (Alpha)"=&FloatFromCurve(
      FloatCurve=DataInterface() {propertyValues {Curve {keys [
          {arriveTangent=6.4019, interpMode=RCIM_Cubic, leaveTangent=6.4019, tangentMode=RCTM_User, time=0, value=0}
          {interpMode=RCIM_Cubic, leaveTangent=0.002, time=0.2, value=0.7}
         ]}}}
      Scale Curve=0.1
     )
    )) {isSetParameters=true}
   /SetVariables(Particles.CentralFlareSize:Vector2f=&Multiply_Vector2D_ByFloat(Float=1.5, Vector2D=$Particles.SpriteSize)) {isSetParameters=true}
   /SetVariables(Particles.CentralFlareColor:LinearColor=&MakeLinearColorFromVectorAndFloat(
     "Vector (RGB)"=&MakeVectorFromLinearColorRGB(Color=&ColorFromCurve(
       ColorCurve=DataInterface() {propertyValues {
         AlphaCurve {keys [<at 0 1 RCIM_Cubic>]}
         BlueCurve {keys [<at 0 0.5 RCIM_Cubic>,<at 1 2 RCIM_Cubic>]}
         GreenCurve {keys [<at 0 0.3303 RCIM_Cubic>,<at 1 1.2751 RCIM_Cubic>]}
         RedCurve {keys [<at 0 0 RCIM_Cubic>,<at 1 0 RCIM_Cubic>]}
        }}
       CurveIndex=$System.SmoothLeanAlpha
      ))
     "Float (Alpha)"=&Multiply_Float(
      A=&FloatFromCurve(FloatCurve=DataInterface() {propertyValues {Curve {keys [
           {arriveTangent=6.4019, interpMode=RCIM_Cubic, leaveTangent=6.4019, tangentMode=RCTM_User, time=0, value=0}
           {interpMode=RCIM_Cubic, leaveTangent=0.002, time=0.2, value=0.7}
          ]}}})
      B=&Multiply_Float(
       A=&FNTimeOfDayScalar(Night Scale=0.5, Morning Scale=0.5, Evening Scale=0.5)
       B=&FN_FrontEndFloatModifier(Gameplay=1, Front End=0.5)
      )
     )
    )) {isSetParameters=true}
   &FN_Light_Attributes(
    Color=$Particles.CentralFlareColor
    Create Color Attribute=true
    Create Exponent Attribute=true
    Create Radius Attribute=true
    Exponent=1
    Radius=&Lerp_Float(A=30, B=15, Alpha=$Particles.HandFootAlpha)
    Scale Color=&Lerp_Float(A=100, B=100, Alpha=$System.SmoothLeanAlpha)
   )
   &DynamicMaterialParameters(
    Index 1 Param 1=0.3
    Index 1 Param 3=1
    Module.Index1Float0Write=true
    Module.Index1Float1Write=true
    Module.Index1Float2Write=true
    Module.Index1Float3Write=true
    Module.Index3Float0Write=true
    Module.Index3Float1Write=true
    Module.Index3Float2Write=true
    Module.Index3Float3Write=true
    Write Parameter Index 0=::NewEnumerator0
    Write Parameter Index 1=::NewEnumerator1
    Write Parameter Index 3=::NewEnumerator1
   )
   /SetVariables(Particles.MeshOrientation=&OrientMeshToVector(Look At Direction=$Output.FollowRibbon.DirectionVector, Reference Axis Vector=<0 0 1>)) {isSetParameters=true}
   &FN_Sprite_Attributes(
    Alignment:Vector3f=<1 0 0>
    Create Alignment Attribute=true
    Create Sprite Rotation Attribute=true
    Create SpriteSize Attribute=true
    SpriteSize=&Multiply_Vector2D(A=$Particles.CentralFlareSize, B=<0.15 1.5>)
    Write to Module Namespace=true
   )
   &FN_Mesh_Attributes(
    Color=&MakeLinearColorFromVectorAndFloat(
     "Vector (RGB)"=&FN_FrontEndVector3Modifier(
      Gameplay=&MakeVectorFromLinearColorRGB(Color=<rgba 0 3.4876 5 1>)
      Front End=&MakeVectorFromLinearColorRGB(Color=<rgba 0 1.3951 2 1>)
     )
     "Float (Alpha)"=1
    )
    Create Color Attribute=true
    Create Orientation Attribute=true
    Create Scale Attribute=true
    Orientation:Quat4f=&MultiplyQuaternion(
     Quaternion A=$Particles.MeshOrientation
     Quaternion B=&MakeQuaternion(Quaternion From=0, Axis=<1 0 0>, Angle=180, From=?, To=?, Angle Type=::NewEnumerator0, Normalize Vectors=false)
    )
    Scale=&Multiply_VectorByFloat(
     Vector=&Lerp_Vector(A=<0.1 0.1 0.2>, B=<0.075 0.075 0.06>, Alpha=$Particles.HandFootAlpha)
     Float=$System.SmoothLeanAlpha
    )
    Write to Module Namespace=true
   )
   /SetVariables(Particles.JetPos:NiagaraPosition=&AddVectorToPosition(
     Position=$Particles.Position
     Vector=&Multiply_VectorByFloat(Vector=$Output.FollowRibbon.DirectionVector, Float=1)
    )) {isSetParameters=true}
  ]
  rendererClasses [
   &NiagaraSpriteRendererProperties
   &NiagaraLightRendererProperties
   &NiagaraMeshRendererProperties
  ]
  renderers [
   {rendererClass=&NiagaraSpriteRendererProperties}
   {rendererClass=&NiagaraSpriteRendererProperties, rendererIndex=1}
   {rendererClass=&NiagaraLightRendererProperties, rendererIndex=2}
   {rendererClass=&NiagaraSpriteRendererProperties, rendererIndex=3}
   {rendererClass=&NiagaraMeshRendererProperties, rendererIndex=4}
   {rendererClass=&NiagaraMeshRendererProperties, rendererIndex=5}
  ]
 }
}
summary {
 systemName=NS_Contrail_SpatialTravelWatt
 emitters [
  {emitterName=NE_DerezRibbons, rendererClasses [&NiagaraRibbonRendererProperties]}
  {emitterName=NE_DerezCubes, rendererClasses [&NiagaraMeshRendererProperties]}
  {
   emitterName=NE_Glow
   rendererClasses [
    &NiagaraSpriteRendererProperties
    &NiagaraLightRendererProperties
    &NiagaraMeshRendererProperties
   ]
  }
 ]
 userVariables [
  User.bIsFrontEnd {type=NiagaraBool}
  User.bIsFrontEndPreview {type=NiagaraBool}
  User.IsFrontEndPreview {type=NiagaraBool}
  User.ParaglideLeanAlpha {type=NiagaraFloat, defaultValue=0, description=""}
 ]
}
